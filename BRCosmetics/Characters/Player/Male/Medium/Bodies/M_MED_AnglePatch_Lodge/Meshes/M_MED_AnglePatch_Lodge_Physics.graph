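# Physics asset for the male medium AnglePatch Lodge player body, set up for dangling left and right loincloth chains.
# Capsules cover pelvis, lower spine, thighs and calves, all joined by free constraints.
# Each cloth chain runs from a small master body, free on the pelvis, through five zero-length capsules (effectively spheres) offset sideways from their bones.
# Chain links allow moderate swing on one axis, little on the other and no twist.
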PhysicsAsset /BRCosmetics/Characters/Player/Male/Medium/Bodies/M_MED_AnglePatch_Lodge/Meshes/M_MED_AnglePatch_Lodge_Physics
bodies {
 calf_l [{shapeType=Capsule, length=26.1655, radius=12.8881, shapeName=calf_l_capsule, center=<-16.3343 0 0>, rotation=<rot 90 0 0>}]
 calf_r [{shapeType=Capsule, length=26.1655, radius=12.8881, shapeName=calf_l_capsule, center=<16.3343 0 0>, rotation=<rot 90 0 180>}]
 dyn_loinCloth_l_01 [{shapeType=Capsule, length=0, radius=8.5, shapeName=dyn_loinCloth_l_01_capsule, center=<3.6911 -8.5 0>}]
 dyn_loinCloth_l_02 [{shapeType=Capsule, length=0, radius=8.5, shapeName=dyn_loinCloth_l_02_capsule, center=<3.6911 -8.5 0>}]
 dyn_loinCloth_l_03 [{shapeType=Capsule, length=0, radius=8.5, shapeName=dyn_loinCloth_l_03_capsule, center=<3.6911 -8.5 0>}]
 dyn_loinCloth_l_04 [{shapeType=Capsule, length=0, radius=10, shapeName=dyn_loinCloth_l_04_capsule, center=<3.6911 -8.5 0>}]
 dyn_loinCloth_l_05 [{shapeType=Capsule, length=0, radius=10, shapeName=dyn_loinCloth_l_05_capsule, center=<5.6347 -8.5 0>}]
 dyn_loinCloth_l_MSTR [{shapeType=Capsule, length=0.505, radius=0.505, shapeName=dyn_loinCloth_l_MSTR_capsule}]
 dyn_loinCloth_r_01 [{shapeType=Capsule, length=0, radius=8.5, shapeName=dyn_loinCloth_r_01_capsule, center=<3.6911 -8.5 0>}]
 dyn_loinCloth_r_02 [{shapeType=Capsule, length=0, radius=8.5, shapeName=dyn_loinCloth_r_02_capsule, center=<3.6911 -8.5 0>}]
 dyn_loinCloth_r_03 [{shapeType=Capsule, length=0, radius=8.5, shapeName=dyn_loinCloth_r_03_capsule, center=<3.6911 -8.5 0>}]
 dyn_loinCloth_r_04 [{shapeType=Capsule, length=0, radius=10, shapeName=dyn_loinCloth_r_04_capsule, center=<3.6911 -8.5 0>}]
 dyn_loinCloth_r_05 [{shapeType=Capsule, length=0, radius=10, shapeName=dyn_loinCloth_r_05_capsule, center=<5.6347 -8.5 0>}]
 dyn_loinCloth_r_MSTR [{shapeType=Capsule, length=0.505, radius=0.505, shapeName=dyn_loinCloth_r_MSTR_capsule}]
 pelvis [{shapeType=Capsule, length=13.0937, radius=11.063, shapeName=pelvis_capsule, center=<0 -0.365 0>}]
 spine_01 [{shapeType=Capsule, length=11.5, radius=10.005, shapeName=spine_01_capsule}]
 thigh_l [{shapeType=Capsule, length=31.1296, radius=11.5925, shapeName=thigh_l_capsule, center=<-14.2622 0 0>, rotation=<rot 90 0 0>}]
 thigh_r [{shapeType=Capsule, length=31.1296, radius=11.5925, shapeName=thigh_l_capsule, center=<14.2622 0 0>, rotation=<rot 90 0 180>}]
}
constraints [
 {
  bone1Name=dyn_loinCloth_l_MSTR
  bone2Name=pelvis
  swing1LimitDegrees=45
  swing1Motion=Free
  swing2LimitDegrees=45
  swing2Motion=Free
  twistLimitDegrees=45
  twistMotion=Free
 }
 {bone1Name=dyn_loinCloth_l_01, bone2Name=dyn_loinCloth_l_MSTR, swing1LimitDegrees=20, swing2LimitDegrees=5, twistLimitDegrees=0}
 {bone1Name=dyn_loinCloth_l_02, bone2Name=dyn_loinCloth_l_01, swing1LimitDegrees=20, swing2LimitDegrees=5, twistLimitDegrees=0}
 {bone1Name=dyn_loinCloth_l_03, bone2Name=dyn_loinCloth_l_02, swing1LimitDegrees=20, swing2LimitDegrees=5, twistLimitDegrees=0}
 {bone1Name=dyn_loinCloth_l_04, bone2Name=dyn_loinCloth_l_03, swing1LimitDegrees=20, swing2LimitDegrees=2, twistLimitDegrees=0}
 {bone1Name=dyn_loinCloth_l_05, bone2Name=dyn_loinCloth_l_04, swing1LimitDegrees=20, swing2LimitDegrees=0, twistLimitDegrees=0}
 {
  bone1Name=dyn_loinCloth_r_MSTR
  bone2Name=pelvis
  swing1LimitDegrees=45
  swing1Motion=Free
  swing2LimitDegrees=45
  swing2Motion=Free
  twistLimitDegrees=45
  twistMotion=Free
 }
 {bone1Name=dyn_loinCloth_r_01, bone2Name=dyn_loinCloth_r_MSTR, swing1LimitDegrees=20, swing2LimitDegrees=5, twistLimitDegrees=0}
 {bone1Name=dyn_loinCloth_r_02, bone2Name=dyn_loinCloth_r_01, swing1LimitDegrees=20, swing2LimitDegrees=5, twistLimitDegrees=0}
 {bone1Name=dyn_loinCloth_r_03, bone2Name=dyn_loinCloth_r_02, swing1LimitDegrees=20, swing2LimitDegrees=5, twistLimitDegrees=0}
 {bone1Name=dyn_loinCloth_r_04, bone2Name=dyn_loinCloth_r_03, swing1LimitDegrees=20, swing2LimitDegrees=2, twistLimitDegrees=0}
 {bone1Name=dyn_loinCloth_r_05, bone2Name=dyn_loinCloth_r_04, swing1LimitDegrees=20, swing2LimitDegrees=0, twistLimitDegrees=0}
 {
  bone1Name=spine_01
  bone2Name=pelvis
  swing1LimitDegrees=45
  swing1Motion=Free
  swing2LimitDegrees=45
  swing2Motion=Free
  twistLimitDegrees=45
  twistMotion=Free
 }
 {
  bone1Name=thigh_l
  bone2Name=pelvis
  swing1LimitDegrees=45
  swing1Motion=Free
  swing2LimitDegrees=45
  swing2Motion=Free
  twistLimitDegrees=45
  twistMotion=Free
 }
 {
  bone1Name=calf_l
  bone2Name=thigh_l
  swing1LimitDegrees=45
  swing1Motion=Free
  swing2LimitDegrees=45
  swing2Motion=Free
  twistLimitDegrees=45
  twistMotion=Free
 }
 {
  bone1Name=calf_r
  bone2Name=thigh_r
  swing1LimitDegrees=45
  swing1Motion=Free
  swing2LimitDegrees=45
  swing2Motion=Free
  twistLimitDegrees=45
  twistMotion=Free
 }
 {
  bone1Name=thigh_r
  bone2Name=pelvis
  swing1LimitDegrees=45
  swing1Motion=Free
  swing2LimitDegrees=45
  swing2Motion=Free
  twistLimitDegrees=45
  twistMotion=Free
 }
]
