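# Post-process style AnimBP for the AnglePatch Lodge male body, adding secondary cloth motion to the pose copied from its parent mesh.
# A hang-glider-aware Control Rig feeds blended loincloth rotations for tactical sprint, then back-cloth RigidBody simulation in pelvis space; a second RigidBody node is unconnected.
# Event updates tune gravity, simulation scales and rig alpha.
# Cape and backpack attach bones collapse when the front end shows the Date glider.
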AnimBlueprint /BRCosmetics/Characters/Player/Male/Medium/Bodies/M_MED_AnglePatch_Lodge/Meshes/M_MED_AnglePatch_Lodge_AnimBP
target_skeleton=&/M_MED_AnglePatch_Lodge_Skeleton
graphs [
 AnimGraph {nodes [
   n1 = Get FRONTaccSclVals() -> (FRONTaccSclVals:Vector)
   RigidBody!(
    ComponentPose
    Alpha=Get RBANTacticalAlpha() -> (RBANTacticalAlpha:double)
    OverrideWorldGravity:Vector=Get Gravity Override()
    ExternalForce
    ComponentLinearAccScale:Vector=@n1
    ComponentLinearVelScale:Vector=@n1
   ) -> (Pose) {
    bClampLinearTranslationLimitToRefPose=true
    bOverrideWorldGravity=true
    overridePhysicsAsset=&/M_MED_AnglePatch_Lodge_Physics
    simulationSpace=BaseBoneSpace
    alphaBoolBlend {blendOption=Linear}
    baseBoneRef {boneName=pelvis}
    componentAppliedLinearAccClamp=<5000 5000 5000>
    componentLinearAccScale=<0.5 0.5 0.5>
    componentLinearVelScale=<0.5 0.5 0.5>
   }
   Output Pose(Result=Use cached pose 'PostSimCache'()) {layerGroup=None}
   n2 = Control Rig!(
    Source=Copy Pose From Mesh() {bUseAttachedParent=true}
    Alpha=Get RBANTacticalAlpha()
    isSkydiving:bool=Get isHangGlider() -> (isHangGlider:bool)
   ) -> (Pose) {
    alphaBoolBlend {blendOption=Linear}
    controlRigAssetReference {blueprintRigClass=&/M_MED_AnglePatch_Lodge_CtrlRig.M_MED_AnglePatch_Lodge_CtrlRig_C}
   }
   Save cached pose 'CRCache'(Pose=@n2) {cacheName=CRCache}
   n3 = Get bIsTacticalSprinting()
   n4 = "Transform (Modify) Bone - Bone: dyn_loinCloth_l_MSTR"!(
    ComponentPose=Local To Component(LocalPose=Use cached pose 'CRCache'())
    bAlphaBoolEnabled:bool=@n3
    Translation="0,0,0"
    Rotation="0,-50,0"
    Scale="1,1,1"
   ) -> (Pose) {
    alphaInputType=Bool
    rotationMode=BMM_Replace
    rotationSpace=BCS_BoneSpace
    alphaBoolBlend {blendInTime=0.6, blendOption=Linear, blendOutTime=1}
    boneToModify {boneName=dyn_loinCloth_l_MSTR}
    rotation {yaw=-50}
   }
   n5 = "Transform (Modify) Bone - Bone: dyn_loinCloth_r_MSTR"!(ComponentPose=@n4, bAlphaBoolEnabled:bool=@n3, Translation="0,0,0", Rotation="0,-70,0", Scale="1,1,1") -> (Pose) {
    alphaInputType=Bool
    rotationMode=BMM_Replace
    rotationSpace=BCS_BoneSpace
    alphaBoolBlend {blendInTime=0.6, blendOption=Linear, blendOutTime=1}
    boneToModify {boneName=dyn_loinCloth_r_MSTR}
    rotation {yaw=-70}
   }
   Save cached pose 'PreSimCache'(Pose=Component To Local(ComponentPose=@n5)) {cacheName=PreSimCache}
   n6 = Get BACKaccSclVals() -> (BACKaccSclVals:Vector)
   n7 = RigidBody(
    ComponentPose=Local To Component(LocalPose=Use cached pose 'PreSimCache'())
    OverrideWorldGravity=Get Gravity Override()
    ComponentLinearAccScale=@n6
    ComponentLinearVelScale=@n6
   ) {
    bClampLinearTranslationLimitToRefPose=true
    bOverrideWorldGravity=true
    overridePhysicsAsset=&/M_MED_AnglePatch_Lodge_BackCloth_Physics
    simulationSpace=BaseBoneSpace
    alphaBoolBlend {blendOption=Linear}
    baseBoneRef {boneName=pelvis}
    componentAppliedLinearAccClamp=<5000 5000 5000>
    componentLinearAccScale=<0.5 0.5 0.5>
    componentLinearVelScale=<0.5 0.5 0.5>
   }
   n8 = Get IsDateGlider() -> (IsDateGlider:bool)
   n9 = "Transform (Modify) Bone - Bone: attach_cape"!(ComponentPose=@n7, bAlphaBoolEnabled:bool=@n8, Translation="0,0,0", Rotation="0,0,0", Scale="0,0,0") -> (Pose) {
    alphaInputType=Bool
    scaleMode=BMM_Replace
    scaleSpace=BCS_BoneSpace
    alphaBoolBlend {blendOption=Linear}
    boneToModify {boneName=attach_cape}
    scale=<0 0 0>
   }
   n10 = "Transform (Modify) Bone - Bone: attach_backpack"!(ComponentPose=@n9, bAlphaBoolEnabled:bool=@n8, Translation="0,0,0", Rotation="0,0,0", Scale="0,0,0") -> (Pose) {
    alphaInputType=Bool
    scaleMode=BMM_Replace
    scaleSpace=BCS_BoneSpace
    alphaBoolBlend {blendOption=Linear}
    boneToModify {boneName=attach_backpack}
    scale=<0 0 0>
   }
   Save cached pose 'RBANCache'(Pose=Component To Local(ComponentPose=@n10)) {cacheName=RBANCache}
   Save cached pose 'PostSimCache'(Pose=Use cached pose 'RBANCache'()) {cacheName=PostSimCache}
   Use cached pose 'PreSimCache'()
  ]}
 EventGraph {nodes [
   n1 = Event BlueprintUpdateAnimation()
   > Set Delta Time X(Delta Time X=@n1.DeltaTimeX)
   > n2 = Sequence!(execute) -> (then_0, then_1, then_2:Exec, then_3:Exec)
   n3 = Cast To FortPlayerPawn(execute=@n2.then_3, Object=GetOwningActor())
   n4 = Get ParachuteAttachment!(self:Fort Player Pawn Object Reference=@n3) -> (ParachuteAttachment:Fort Player Parachute Object Reference)
   > Is Valid(InputObject=@n4)
   n5 = Get ParachuteMesh!(self:Fort Player Parachute Object Reference=@n4) -> (ParachuteMesh:Skeletal Mesh Component Budgeted Object Reference)
   > Is Valid(InputObject=@n5)
   n6 = Get SkeletalMesh!(self:Skinned Mesh Component Object Reference=@n5) -> (SkeletalMesh:Skeletal Mesh Object Reference)
   > Is Valid(InputObject=@n6)
   > n7 = Branch(Condition=Get bIsInFrontEnd())
   > Set IsDateGlider(IsDateGlider:bool="Equal (SoftObjectReference)"(
     A:"Soft Object Reference (by ref)"=Make Soft Reference!(Input:Skeletal Mesh Object Reference=@n6) -> (Output:Skeletal Mesh Soft Object Reference)
     B:"Soft Object Reference (by ref)"=Get DateGlider() -> (DateGlider:Skeletal Mesh Soft Object Reference)
    ) -> (ReturnValue:bool=~) {is=title K2Node_PromotableOperator}) -> (Output_Get:bool)
   Set IsDateGlider(execute=@n7.else, IsDateGlider=false)
   n8 = Get PlayerGliderType() -> (PlayerGliderType=HangGlider)
   n9 = Branch(
    execute=@n2.then_2
    Condition=AND Boolean(A="Equal (Enum)"(A:EGliderType Enum=@n8, B:EGliderType Enum=HangGlider), B=Get bIsParachuteOpened())
   )
   > Set isHangGlider(isHangGlider:bool=true) -> (Output_Get:bool)
   Set isHangGlider(execute=@n9.else, isHangGlider=false)
   n10 = Gravity Override State Function DEPRECATED(
    execute=@n2.then_0
    CharacterPartAnimInstance=Self-Reference()
    GameTimeInSeconds=Get Delta Time X()
    GravityOverrideParams=Get Gravity Override Params()
   )
   > Set Gravity Override(Gravity Override=@n10)
   n11 = Branch(execute=@n2.then_1, Condition=Get bIsInFrontEnd())
   > Set FRONTaccSclVals(FRONTaccSclVals:Vector="0.5,0.5,0.5") -> (Output_Get:Vector)
   n12 = Branch(execute=@n11.else, Condition=Get bIsTacticalSprinting())
   > Set FRONTaccSclVals(FRONTaccSclVals="0.1,0.1,0.1")
   > Set BACKaccSclVals(BACKaccSclVals:Vector="1,1,1") -> (Output_Get:Vector)
   > Set RBANTacticalAlpha(RBANTacticalAlpha:double=0.3) -> (Output_Get:double)
   Set FRONTaccSclVals(execute=@n12.else, FRONTaccSclVals="0.5,0.5,0.5")
   > Set BACKaccSclVals(BACKaccSclVals="0.5,0.5,0.5")
   > Set RBANTacticalAlpha(RBANTacticalAlpha=1)
  ]}
]
variables [
 BACKaccSclVals {type=Vector}
 DateGlider {type=SkeletalMesh, default=/BRCosmetics/Gadgets/Assets/VinderTech_GliderChute/Glider_HighTower_Date/Meshes/Glider_HighTower_Date}
 Delta Time X {type=number}
 FRONTaccSclVals {type=Vector}
 Gravity Override {type=Vector}
 Gravity Override Params {type=GravityOverrideParamsStruct}
 IsDateGlider {type=boolean}
 isHangGlider {type=boolean}
 RBANTacticalAlpha {type=number}
]
