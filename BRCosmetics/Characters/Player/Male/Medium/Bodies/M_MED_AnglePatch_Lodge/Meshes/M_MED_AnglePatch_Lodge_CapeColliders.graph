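# Physics asset of collision shapes for the cape on the AnglePatch Lodge male medium player body, judging by its CapeColliders name.
# Capsules cover the pelvis, spine, clavicles and neck, with a sphere on the head, forming a torso volume.
# Further capsules ride on left and right dyn_clothC bone chains.
# Arm, leg and foot bodies are empty, and no constraints exist, so it supplies collision only, not ragdoll joints.
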PhysicsAsset /BRCosmetics/Characters/Player/Male/Medium/Bodies/M_MED_AnglePatch_Lodge/Meshes/M_MED_AnglePatch_Lodge_CapeColliders
bodies {
 calf_l []
 calf_r []
 clavicle_l [{shapeType=Capsule, length=9.0192, radius=9.8871, center=<9.4869 -1.9815 -2.5128>, rotation=<rot 69.3297 1.4476 176.9462>}]
 clavicle_r [{shapeType=Capsule, length=9.0192, radius=9.8871, center=<-9.4869 1.9815 2.5128>, rotation=<rot 69.3297 1.4476 -3.0538>}]
 dyn_clothC_l_02 [{
   shapeType=Capsule
   length=6.1462
   radius=4.6595
   shapeName=dyn_clothC_l_04_capsule
   center=<2.363 2.6905 -0.4211>
   rotation=<rot 0 -2.2678 0.2355>
  }]
 dyn_clothC_l_03 [{
   shapeType=Capsule
   length=6.1462
   radius=4.6595
   shapeName=dyn_clothC_l_04_capsule
   center=<2.363 2.6905 -0.4211>
   rotation=<rot 0 -2.2678 0.2355>
  }]
 dyn_clothC_l_04 [{
   shapeType=Capsule
   length=6.1462
   radius=6.5535
   shapeName=dyn_clothC_l_04_capsule
   center=<0.2018 3.5975 -0.1732>
   rotation=<rot 0 -2.2678 0.2355>
  }]
 dyn_clothC_l_05 [{shapeType=Capsule, length=6.0639, radius=6.8519, shapeName=dyn_clothC_l_05_capsule, center=<3.1658 3.8107 0>}]
 dyn_clothC_r_02 [{
   shapeType=Capsule
   length=6.7536
   radius=4.5551
   shapeName=dyn_clothC_r_04_capsule
   center=<2.376 2.6905 0.2853>
   rotation=<rot 0 -2.2603 -0.1721>
  }]
 dyn_clothC_r_03 [{
   shapeType=Capsule
   length=6.7536
   radius=4.5551
   shapeName=dyn_clothC_r_04_capsule
   center=<2.376 2.6905 0.2853>
   rotation=<rot 0 -2.2603 -0.1721>
  }]
 dyn_clothC_r_04 [{
   shapeType=Capsule
   length=6.7536
   radius=6.449
   shapeName=dyn_clothC_r_04_capsule
   center=<0.2071 3.5975 0.1174>
   rotation=<rot 0 -2.2603 -0.1721>
  }]
 dyn_clothC_r_05 [{shapeType=Capsule, length=6.0639, radius=6.8519, shapeName=dyn_clothC_r_05_capsule, center=<3.1658 3.8107 0>}]
 foot_l []
 foot_r []
 head [{shapeType=Sphere, radius=17.3125, center=<11.028 -0.174 0>}]
 lowerarm_l []
 lowerarm_r []
 neck_02 [{shapeType=Capsule, length=13.4375, radius=7.3175, center=<-0.2642 -0.3408 0>, rotation=<rot 90 -171.1193 -175.3978>}]
 pelvis [{shapeType=Capsule, length=10.4146, radius=12.5983, center=<-2.4887 0.035 0>}]
 spine_01 [{shapeType=Capsule, length=7.9438, radius=13.5586, center=<-1.7965 1.6318 0>}]
 spine_02 [{shapeType=Capsule, length=7.3264, radius=12.6474, center=<-1.711 0.2698 0.0001>}]
 spine_03 [{shapeType=Capsule, length=8.2107, radius=13.7031, center=<1.7341 -0.2574 0.0001>}]
 spine_04 [{shapeType=Capsule, length=9.4404, radius=14.1325, center=<4.2405 -1.3646 0.0002>}]
 spine_05 [
  {shapeType=Capsule, length=7.5761, radius=16.1415, center=<-0.4501 2.6408 0>}
  {shapeType=Capsule, length=13.9518, radius=13.9425, center=<-0.2916 0.4452 0>}
 ]
 thigh_l []
 thigh_r []
 upperarm_l []
 upperarm_r []
}
constraints []
